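# Sound cue for the explosion of the Figment grenade launcher's grenade, mixing a close and a distant layer under separate sound classes.
# The close layer blends two waves through a modulator and an overridden attenuation with a steep custom falloff and distance-scaled distant-reverb submix send.
# The distant layer plays a pitch-randomized wave through a distant explosion attenuation.
# A shared weapon-explosion concurrency set limits overlapping instances.
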SoundCue /Figment_Launcher_Grenade_Athena/Sounds/Grenade_Launcher_Expl_Figment_Cue
root=Mixer_4
nodes [
 {
  id=Attenuation_0
  children [Modulator_1]
  properties {
   attenuationSettings=/Game/Sounds/Attenuation/Fort_Explosion_Distant_Athena_Att
   attenuationOverrides {falloffDistance=3000, attenuationShapeExtents {x=1000}}
  }
 }
 {
  id=Attenuation_4
  children [Mixer_0]
  properties {
   attenuationSettings=/Game/Sounds/Attenuation/Fort_Explosion_Medium_Athena_Att
   bOverrideAttenuation=true
   attenuationOverrides {
    bEnableSubmixSends=true
    distanceAlgorithm=Custom
    falloffDistance=27500
    nonSpatializedRadiusEnd=150
    nonSpatializedRadiusMode=Direct2D
    nonSpatializedRadiusStart=350
    reverbDistanceMax=1024
    reverbDistanceMin=0
    reverbWetLevelMax=0
    reverbWetLevelMin=1
    stereoSpread=1000
    attenuationShapeExtents {x=0}
    customAttenuationCurve {editorCurveData {keys [<at 0 1>, {arriveTangent=-0.2335, interpMode=RCIM_Cubic, leaveTangent=-0.2335, tangentMode=RCTM_User, time=0.0616, value=1}, <at 0.19 0>]}}
    submixSendSettings [{
      bEnableHPFCutoff=false
      bEnableLPFCutoff=false
      disableManualSendClamp=false
      hPFCutoff=20
      lPFCutoff=20000
      maxSendDistance=3500
      maxSendLevel=0.3
      minSendDistance=500
      minSendLevel=0.1
      sendLevel=0.2
      sendLevelControlMethod=Linear
      soundSubmix=&Fortnite_Reverb_Distant
     }]
   }
  }
 }
 {id=Mixer_0, children [WavePlayer_4,WavePlayer_0], properties {inputVolume [0.4,1.2]}}
 {id=Mixer_4, children [null,SoundClass_3,SoundClass_0], properties {inputVolume [1,1,1]}}
 {id=Modulator_0, children [Attenuation_4]}
 {id=Modulator_1, children [WavePlayer_1], properties {pitchMax=1.1, pitchMin=0.9}}
 {
  id=SoundClass_0
  children [Attenuation_0]
  properties {soundClassOverride=&Fort_WeaponRanged_Fire_Distant}
 }
 {id=SoundClass_3, children [Modulator_0], properties {soundClassOverride=&Fort_WeaponRanged_Fire_3P_PSM}}
 {id=WavePlayer_0, properties {soundWaveAssetPtr=&/explosion_sharp_mono_01}}
 {id=WavePlayer_1, properties {soundWaveAssetPtr=&/explosion_sharp_distant_01}}
 {id=WavePlayer_4, properties {soundWaveAssetPtr=&/Explosion_GrenadeProjectile_01_E}}
]
properties {
 maxDistance=27500
 volumeMultiplier=1
 concurrencyOverrides {maxCount=4, resolutionRule=StopOldest, volumeScale=0.5}
 concurrencySet [/Game/Sounds/Concurrency_Rules/Weapon_Explosion_Conc]
}
