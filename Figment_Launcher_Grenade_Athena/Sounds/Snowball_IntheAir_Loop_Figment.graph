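# Looping in-flight sound cue for the Figment snowball grenade launcher projectile, classed as a large explosive sound.
# It randomly picks one of two looping waves with slight upward pitch variation, then an envelope raises its volume over about two seconds.
# An intensified Doppler node feeds a mixer, whose second input is empty, into RPG explosion attenuation.
# A concurrency override caps simultaneous instances to a few.
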SoundCue /Figment_Launcher_Grenade_Athena/Sounds/Snowball_IntheAir_Loop_Figment
root=Attenuation_1
nodes [
 {id=Attenuation_1, children [Mixer_1], properties {attenuationSettings=/Game/Sounds/Attenuation/Fort_Explosion_RPG_Att}}
 {id=Doppler_1, children [Enveloper_1], properties {dopplerIntensity=4}}
 {
  id=Enveloper_1
  children [Modulator_2]
  properties {
   pitchCurve {editorCurveData {keys [{time=0}]}}
   volumeCurve {editorCurveData {keys [<at 0.0002 0.5183>,<at 0.9977 0.6079>,<at 1.905 0.9819>]}}
  }
 }
 {id=Mixer_1, children [Doppler_1,null], properties {inputVolume [1,1]}}
 {id=Modulator_2, children [Random_2], properties {pitchMax=1.2, pitchMin=1}}
 {id=Random_2, children [WavePlayer_3,WavePlayer_6], properties {weights [1,1]}}
 {
  id=WavePlayer_3
  properties {bLooping=true, soundWaveAssetPtr=&/WPN_SnowBall_Launcher_IntheAir_LOOP_01}
 }
 {
  id=WavePlayer_6
  properties {bLooping=true, soundWaveAssetPtr=&/WPN_SnowBall_Launcher_IntheAir_LOOP_02}
 }
]
properties {
 attenuationSettings=&Fort_Explosion_RPG_Att
 bOverrideConcurrency=true
 maxDistance=13000
 soundClassObject=/Game/Sounds/FortSoundClassesAndModes/Fort_Explosives_Large_PSM
 volumeMultiplier=0.4246
 concurrencyOverrides {maxCount=3}
}
