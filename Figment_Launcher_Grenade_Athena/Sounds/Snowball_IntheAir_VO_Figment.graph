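# Sound cue for the Figment snowball launcher, playing a "Weee" voice line, apparently while the projectile is in the air.
# A random pick of three looping waves feeds a modulator with fixed pitch, then an enveloper and a short random delay.
# The enveloper fades volume in quickly, then down to silence over several seconds.
# Medium pickaxe attenuation, the large explosives sound class and a small concurrency cap govern playback.
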SoundCue /Figment_Launcher_Grenade_Athena/Sounds/Snowball_IntheAir_VO_Figment
root=Delay_1
nodes [
 {id=Delay_1, children [Enveloper_6], properties {delayMax=0.3, delayMin=0.2}}
 {
  id=Enveloper_6
  children [Modulator_1]
  properties {
   pitchCurve {editorCurveData {keys [{time=0}]}}
   volumeCurve {editorCurveData {keys [
      {interpMode=RCIM_Cubic, tangentMode=RCTM_User, time=0, value=0}
      {interpMode=RCIM_Cubic, tangentMode=RCTM_User, time=0.25, value=1}
      <at 5 0>
     ]}}
  }
 }
 {id=Modulator_1, children [Random_0], properties {pitchMax=1, pitchMin=1}}
 {id=Random_0, children [WavePlayer_7,WavePlayer_8,WavePlayer_9], properties {weights [1,1,1]}}
 {id=WavePlayer_7, properties {bLooping=true, soundWaveAssetPtr=&/WPN_SnowBall_Launcher_VO_Weee_01}}
 {id=WavePlayer_8, properties {bLooping=true, soundWaveAssetPtr=&/WPN_SnowBall_Launcher_VO_Weee_12}}
 {id=WavePlayer_9, properties {bLooping=true, soundWaveAssetPtr=&/WPN_SnowBall_Launcher_VO_Weee_13}}
]
properties {
 attenuationSettings=/Game/Sounds/Attenuation/Fort_Weapon_Pickaxe_Medium_Att
 bOverrideConcurrency=true
 maxDistance=3000
 soundClassObject=/Game/Sounds/FortSoundClassesAndModes/Fort_Explosives_Large_PSM
 volumeMultiplier=0.4246
 concurrencyOverrides {maxCount=3}
}
