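# Sound cue for the in-air loop of the Figment grenade launcher's projectile.
# A small grenade loop wave feeds a mono attenuation node with logarithmic falloff and HRTF spatialization, then a Doppler node for pitch shift in motion.
# A sound class node at the root routes it as a medium explosive.
# A shared explosive fuse concurrency set governs simultaneous instances.
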SoundCue /Figment_Launcher_Grenade_Athena/Sounds/Weapon_LAU_Grenade_Launcher_RocketInAir_Figment_Cue
root=SoundClass_2
nodes [
 {
  id=Attenuation_0
  children [WavePlayer_4]
  properties {
   attenuationSettings=&Fort_Weapon_Projectile_Medium_Mono_Athena_Att
   bOverrideAttenuation=true
   attenuationOverrides {
    distanceAlgorithm=Logarithmic
    falloffDistance=1536
    spatializationAlgorithm=SPATIALIZATION_HRTF
    stereoSpread=0
    attenuationShapeExtents {x=512}
   }
  }
 }
 {id=Doppler_2, children [Attenuation_0], properties {dopplerIntensity=2}}
 {id=SoundClass_2, children [Doppler_2], properties {soundClassOverride=/Game/Sounds/FortSoundClassesAndModes/Fort_Explosives_Medium_PSM}}
 {id=WavePlayer_4, properties {soundWaveAssetPtr=&/Grenade_InAirLoop_Small_01}}
]
properties {
 maxDistance=2048
 volumeMultiplier=1
 attenuationOverrides {falloffDistance=5000, attenuationShapeExtents {x=1000}}
 concurrencyOverrides {
  bVolumeScaleCanRelease=true
  maxCount=2
  resolutionRule=StopFarthestThenPreventNew
  volumeScale=0.2
  volumeScaleAttackTime=0.2
  volumeScaleReleaseTime=2
 }
 concurrencySet [/Game/Sounds/Concurrency_Rules/Athena_Explosive_Fuse_Conc]
}
