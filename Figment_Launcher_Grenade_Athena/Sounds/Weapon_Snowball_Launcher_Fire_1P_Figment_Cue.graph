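# Sound cue for the first-person firing sound of the Figment grenade launcher, using snowball launcher audio.
# A random node picks between two equally weighted launcher wave variants and feeds a modulator at the root.
# Playback uses first-person projectile attenuation, the first-person ranged weapon fire sound class, and a gunfire concurrency set.
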SoundCue /Figment_Launcher_Grenade_Athena/Sounds/Weapon_Snowball_Launcher_Fire_1P_Figment_Cue
root=Modulator_15
nodes [
 {id=Modulator_15, children [Random_1]}
 {id=Random_1, children [WavePlayer_12, WavePlayer_13], properties {weights [1,1]}}
 {id=WavePlayer_12, properties {soundWaveAssetPtr=&/WPN_SnowBall_Launcher_v1}}
 {id=WavePlayer_13, properties {soundWaveAssetPtr=&/WPN_SnowBall_Launcher_v2}}
]
properties {
 attenuationSettings=&Fort_Weapon_Projectile_FirstPerson_Athena_Att
 maxDistance=2097152
 soundClassObject=&Fort_WeaponRanged_Fire_1P_PSM
 volumeMultiplier=0.8
 attenuationOverrides {falloffDistance=5000, attenuationShapeExtents {x=1000}}
 concurrencyOverrides {bLimitToOwner=true, maxCount=5, resolutionRule=StopOldest}
 concurrencySet [&Weapon_Gunfire_FirstPerson_Conc]
}
