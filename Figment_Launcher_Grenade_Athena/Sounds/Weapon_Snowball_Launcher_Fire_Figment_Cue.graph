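# Sound cue for the Figment snowball launcher's fire sound, layered by distance.
# A mixer blends a close branch in the third-person fire sound class with a distant-class branch that mixes medium and far layers, the far one boosted.
# Each layer has its own attenuation and modulator; close and medium randomly pick between two waves, while far uses one.
# Local and global gunfire concurrency sets are assigned to the cue.
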SoundCue /Figment_Launcher_Grenade_Athena/Sounds/Weapon_Snowball_Launcher_Fire_Figment_Cue
root=Mixer_1
nodes [
 {
  id=Attenuation_10
  children [Modulator_14]
  properties {attenuationSettings=&Fort_Weapon_Projectile_Medium_Mono_Athena_Att}
 }
 {
  id=Attenuation_11
  children [Modulator_13]
  properties {attenuationSettings=&Fort_Weapon_Projectile_Distant_Mono_Athena_Att}
 }
 {id=Attenuation_9, children [Modulator_15], properties {attenuationSettings=&Fort_Weapon_Projectile_Close_Athena_Att}}
 {id=Mixer_1, children [SoundClass_7,SoundClass_6], properties {inputVolume [1,1]}}
 {id=Mixer_2, children [Attenuation_10, Attenuation_11], properties {inputVolume [1,1.25]}}
 {id=Modulator_13, children [WavePlayer_5], properties {pitchMax=1.1}}
 {id=Modulator_14, children [Random_0], properties {pitchMax=1.1}}
 {id=Modulator_15, children [Random_1]}
 {id=Random_0, children [WavePlayer_10, WavePlayer_11], properties {weights [1,1]}}
 {id=Random_1, children [WavePlayer_12, WavePlayer_13], properties {weights [1,1]}}
 {id=SoundClass_6, children [Mixer_2], properties {soundClassOverride=&Fort_WeaponRanged_Fire_Distant}}
 {
  id=SoundClass_7
  children [Attenuation_9]
  properties {soundClassOverride=&Fort_WeaponRanged_Fire_3P_PSM}
 }
 {id=WavePlayer_10, properties {soundWaveAssetPtr=&/WPN_SnowBall_Launcher_Fire_Med_Distant_v2_01}}
 {id=WavePlayer_11, properties {soundWaveAssetPtr=&/WPN_SnowBall_Launcher_Fire_Med_Distant_v2_02}}
 {id=WavePlayer_12, properties {soundWaveAssetPtr=&/WPN_SnowBall_Launcher_v1}}
 {id=WavePlayer_13, properties {soundWaveAssetPtr=&/WPN_SnowBall_Launcher_v2}}
 {id=WavePlayer_5, properties {soundWaveAssetPtr=&/WPN_SnowBall_Launcher_Fire_Far_Distance_01}}
]
properties {
 maxDistance=27500
 volumeMultiplier=0.8
 attenuationOverrides {falloffDistance=5000, attenuationShapeExtents {x=1000}}
 concurrencyOverrides {bLimitToOwner=true, maxCount=3, resolutionRule=StopOldest}
 concurrencySet [&Weapon_Gunfire_Local_Conc, &Weapon_Gunfire_Global_Conc]
}
