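# Sound cue for the start of the Figment grenade launcher's reload, using snowball launcher sounds, routed as weapon foley.
# A mixer layers a generic launcher reload immediately, the snowball reload after a short random delay, and a pitched-up giggle voice slightly later.
# Each layer passes through its own modulator.
# Reload attenuation applies, and a concurrency override allows one instance, resolved by stopping the farthest first.
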SoundCue /Figment_Launcher_Grenade_Athena/Sounds/Weapon_Snowball_Launcher_Reload_Start_Figment_Cue
root=Mixer_0
nodes [
 {id=Delay_0, children [WavePlayer_0], properties {delayMax=0.95, delayMin=0.85}}
 {id=Delay_1, children [WavePlayer_2], properties {delayMax=0.65, delayMin=0.5}}
 {id=Mixer_0, children [Modulator_0,Modulator_1,Modulator_2], properties {inputVolume [1,1,1]}}
 {id=Modulator_0, children [WavePlayer_1]}
 {id=Modulator_1, children [Delay_1]}
 {id=Modulator_2, children [Delay_0], properties {pitchMax=1.35, pitchMin=1.15}}
 {id=WavePlayer_0, properties {soundWaveAssetPtr=&/Reload_SnowBall_VO_Giggle_01}}
 {id=WavePlayer_1, properties {soundWaveAssetPtr=&/weapons_launchers_reload_01}}
 {id=WavePlayer_2, properties {soundWaveAssetPtr=&/WPN_SnowBall_Launcher_Reload}}
]
properties {
 attenuationSettings=&Fort_Weapon_Reload_Att
 bOverrideConcurrency=true
 maxDistance=2250
 soundClassObject=&Fort_WeaponFoley
 volumeMultiplier=0.501
 attenuationOverrides {falloffDistance=5000, attenuationShapeExtents {x=1000}}
 concurrencyOverrides {maxCount=1, resolutionRule=StopFarthestThenPreventNew}
}
